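# Sound cue for the seventh House of Mirrors ambient spot sound, filed under the Creative Speaker gadget's Survival Horror sounds.
# A single wave player plays the Amb_HouseOfMirrors_Spot_07 wave directly, with no randomization, mixing or modulation nodes.
# It uses the shared tile ambient attenuation asset and routes to the SFX sound class.
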SoundCue /Game/Sounds/Creative/Gadgets/Speaker/Survival_Horror/HouseOfMirrors/Amb_HouseOfMirrors_Spot_07_Cue
root=WavePlayer_0
nodes [{
  id=WavePlayer_0
  properties {soundWaveAssetPtr=&/Amb_HouseOfMirrors_Spot_07}
 }]
properties {
 attenuationSettings=/Game/Sounds/Attenuation/Fort_TileAmbient_Att
 maxDistance=10000
 soundClassObject=&SFX
}
